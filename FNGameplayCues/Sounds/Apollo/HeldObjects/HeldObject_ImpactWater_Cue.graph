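# Sound cue that plays a water impact sound for held objects, classed as an impact sound with held-object attenuation.
# A switch on the AbilityLevel integer parameter selects the small, medium or large size set, with medium also serving as the default.
# Each set randomly picks one of three equally weighted waves, and the result passes through a modulator.
# Playback is limited by the player melee weapon concurrency set.
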SoundCue /FNGameplayCues/Sounds/Apollo/HeldObjects/HeldObject_ImpactWater_Cue
root=Modulator_0
nodes [
 {id=Modulator_0, children [Switch_0]}
 {id=Random_1, children [WavePlayer_23, WavePlayer_24, WavePlayer_25], properties {weights [1,1,1]}}
 {id=Random_2, children [WavePlayer_17, WavePlayer_18, WavePlayer_19], properties {weights [1,1,1]}}
 {id=Random_3, children [WavePlayer_13, WavePlayer_14, WavePlayer_15], properties {weights [1,1,1]}}
 {id=Switch_0, children [Random_2,Random_1,Random_2,Random_3], properties {intParameterName=AbilityLevel}}
 {id=WavePlayer_13, properties {soundWaveAssetPtr=&/HeldObjects_Large_ImpactWater_01}}
 {id=WavePlayer_14, properties {soundWaveAssetPtr=&/HeldObjects_Large_ImpactWater_02}}
 {id=WavePlayer_15, properties {soundWaveAssetPtr=&/HeldObjects_Large_ImpactWater_03}}
 {id=WavePlayer_17, properties {soundWaveAssetPtr=&/HeldObjects_Medium_ImpactWater_01}}
 {id=WavePlayer_18, properties {soundWaveAssetPtr=&/HeldObjects_Medium_ImpactWater_02}}
 {id=WavePlayer_19, properties {soundWaveAssetPtr=&/HeldObjects_Medium_ImpactWater_03}}
 {id=WavePlayer_23, properties {soundWaveAssetPtr=/Game/Sounds/Apollo/HeldObjects/HeldObjects_Small_ImpactWater_01}}
 {id=WavePlayer_24, properties {soundWaveAssetPtr=/Game/Sounds/Apollo/HeldObjects/HeldObjects_Small_ImpactWater_02}}
 {id=WavePlayer_25, properties {soundWaveAssetPtr=/Game/Sounds/Apollo/HeldObjects/HeldObjects_Small_ImpactWater_03}}
]
properties {
 attenuationSettings=/FNGameplayCues/Sounds/Attenuation/Fort_HeldObject_Impact_Att
 maxDistance=1912
 soundClassObject=&Fort_Impacts
 volumeMultiplier=1
 attenuationOverrides {falloffDistance=800, attenuationShapeExtents {x=300}}
 concurrencySet [&Player_Melee_Weapon_Conc]
}
