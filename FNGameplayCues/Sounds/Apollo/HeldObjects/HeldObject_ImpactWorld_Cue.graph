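# Sound cue for held objects hitting the world, with small, medium and large impact variants.
# An AbilityLevel integer switch selects the size, falling back to medium, and each size plays one of three equally weighted random waves.
# A Modulator wraps the switch.
# The cue plays in the impacts sound class with held-object impact attenuation and references the player melee weapon concurrency set.
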SoundCue /FNGameplayCues/Sounds/Apollo/HeldObjects/HeldObject_ImpactWorld_Cue
root=Modulator_1
nodes [
 {id=Modulator_1, children [Switch_0]}
 {id=Random_1, children [WavePlayer_17, WavePlayer_18, WavePlayer_19], properties {weights [1,1,1]}}
 {id=Random_2, children [WavePlayer_12, WavePlayer_13, WavePlayer_14], properties {weights [1,1,1]}}
 {id=Random_3, children [WavePlayer_8, WavePlayer_9, WavePlayer_10], properties {weights [1,1,1]}}
 {id=Switch_0, children [Random_2,Random_1,Random_2,Random_3], properties {intParameterName=AbilityLevel}}
 {id=WavePlayer_10, properties {soundWaveAssetPtr=&/HeldObjects_Large_ImpactSurface_03}}
 {id=WavePlayer_12, properties {soundWaveAssetPtr=&/HeldObjects_Medium_ImpactSurface_01}}
 {id=WavePlayer_13, properties {soundWaveAssetPtr=&/HeldObjects_Medium_ImpactSurface_02}}
 {id=WavePlayer_14, properties {soundWaveAssetPtr=&/HeldObjects_Medium_ImpactSurface_03}}
 {id=WavePlayer_17, properties {soundWaveAssetPtr=&/HeldObjects_Small_ImpactSurface_01}}
 {id=WavePlayer_18, properties {soundWaveAssetPtr=&/HeldObjects_Small_ImpactSurface_02}}
 {id=WavePlayer_19, properties {soundWaveAssetPtr=&/HeldObjects_Small_ImpactSurface_03}}
 {id=WavePlayer_8, properties {soundWaveAssetPtr=&/HeldObjects_Large_ImpactSurface_01}}
 {id=WavePlayer_9, properties {soundWaveAssetPtr=&/HeldObjects_Large_ImpactSurface_02}}
]
properties {
 attenuationSettings=/FNGameplayCues/Sounds/Attenuation/Fort_HeldObject_Impact_Att
 maxDistance=1912
 soundClassObject=&Fort_Impacts
 volumeMultiplier=0.8
 attenuationOverrides {
  bAttenuateWithLPF=true
  distanceAlgorithm=Logarithmic
  falloffDistance=800
  lPFRadiusMax=800
  lPFRadiusMin=400
  attenuationShapeExtents {x=300}
 }
 concurrencyOverrides {bLimitToOwner=true, maxCount=2, resolutionRule=StopOldest}
 concurrencySet [&Player_Melee_Weapon_Conc]
}
