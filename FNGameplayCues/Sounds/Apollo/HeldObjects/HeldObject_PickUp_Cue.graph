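# Sound cue for picking up held objects, playing Small, Large or Huge pickup sounds as player foley.
# After a short fixed delay and a modulator, a switch on the AbilityLevel parameter selects the size, with Large also serving as the default.
# Each size randomly picks among three equally weighted waves.
# Playback uses close-prop attenuation with overrides and a two-instance limit that stops the oldest.
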SoundCue /FNGameplayCues/Sounds/Apollo/HeldObjects/HeldObject_PickUp_Cue
root=Delay_1
nodes [
 {id=Delay_1, children [Modulator_0], properties {delayMax=0.08, delayMin=0.08}}
 {id=Modulator_0, children [Switch_0]}
 {id=Random_0, children [WavePlayer_8, WavePlayer_9, WavePlayer_10], properties {weights [1,1,1]}}
 {id=Random_1, children [WavePlayer_14, WavePlayer_15, WavePlayer_16], properties {weights [1,1,1]}}
 {id=Random_3, children [WavePlayer_5,WavePlayer_6,WavePlayer_7], properties {weights [1,1,1]}}
 {id=Switch_0, children [Random_0,Random_1,Random_0,Random_3], properties {intParameterName=AbilityLevel}}
 {id=WavePlayer_10, properties {soundWaveAssetPtr=&/HeldObjects_Large_Pickup_03}}
 {id=WavePlayer_14, properties {soundWaveAssetPtr=&/HeldObjects_Small_Pickup_01}}
 {id=WavePlayer_15, properties {soundWaveAssetPtr=&/HeldObjects_Small_Pickup_02}}
 {id=WavePlayer_16, properties {soundWaveAssetPtr=&/HeldObjects_Small_Pickup_03}}
 {id=WavePlayer_5, properties {soundWaveAssetPtr=&/HeldObjects_Huge_Pickup_01}}
 {id=WavePlayer_6, properties {soundWaveAssetPtr=&/HeldObjects_Huge_Pickup_02}}
 {id=WavePlayer_7, properties {soundWaveAssetPtr=&/HeldObjects_Huge_Pickup_03}}
 {id=WavePlayer_8, properties {soundWaveAssetPtr=&/HeldObjects_Large_Pickup_01}}
 {id=WavePlayer_9, properties {soundWaveAssetPtr=&/HeldObjects_Large_Pickup_02}}
]
properties {
 attenuationSettings=/Game/Sounds/Attenuation/Fort_Prop_Close_Att
 bOverrideAttenuation=true
 bOverrideConcurrency=true
 maxDistance=1200
 soundClassObject=&Fort_Player_Foley
 volumeMultiplier=0.8
 attenuationOverrides {falloffDistance=800, stereoSpread=500}
 concurrencyOverrides {maxCount=2, resolutionRule=StopOldest}
}
