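# Sound cue for throwing held objects, picking a small, medium or huge throw sound.
# After a short fixed delay and a modulator, a switch on the AbilityLevel integer parameter selects a randomized three-wave set per size, with the medium set also serving as the default.
# It uses the held-object impact attenuation and ranged-weapon sound class, with a concurrency override that prevents new instances at the limit.
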SoundCue /FNGameplayCues/Sounds/Apollo/HeldObjects/HeldObject_Throw_Cue
root=Delay_0
nodes [
 {id=Delay_0, children [Modulator_0], properties {delayMax=0.04, delayMin=0.04}}
 {id=Modulator_0, children [Switch_0]}
 {id=Random_0, children [WavePlayer_15, WavePlayer_16, WavePlayer_17], properties {weights [1,1,1]}}
 {id=Random_2, children [WavePlayer_11, WavePlayer_12, WavePlayer_13], properties {weights [1,1,1]}}
 {id=Random_3, children [WavePlayer_0,WavePlayer_1,WavePlayer_2], properties {weights [1,1,1]}}
 {id=Switch_0, children [Random_2,Random_0,Random_2,Random_3], properties {intParameterName=AbilityLevel}}
 {id=WavePlayer_0, properties {soundWaveAssetPtr=&/HeldObjects_Huge_Throw_01}}
 {id=WavePlayer_1, properties {soundWaveAssetPtr=&/HeldObjects_Huge_Throw_02}}
 {id=WavePlayer_11, properties {soundWaveAssetPtr=&/HeldObjects_Medium_Throw_01}}
 {id=WavePlayer_12, properties {soundWaveAssetPtr=&/HeldObjects_Medium_Throw_02}}
 {id=WavePlayer_13, properties {soundWaveAssetPtr=&/HeldObjects_Medium_Throw_03}}
 {id=WavePlayer_15, properties {soundWaveAssetPtr=&/HeldObjects_Small_Throw_01}}
 {id=WavePlayer_16, properties {soundWaveAssetPtr=&/HeldObjects_Small_Throw_02}}
 {id=WavePlayer_17, properties {soundWaveAssetPtr=&/HeldObjects_Small_Throw_03}}
 {id=WavePlayer_2, properties {soundWaveAssetPtr=&/HeldObjects_Huge_Throw_03}}
]
properties {
 attenuationSettings=/FNGameplayCues/Sounds/Attenuation/Fort_HeldObject_Impact_Att
 bOverrideConcurrency=true
 maxDistance=1912
 soundClassObject=&Fort_WeaponRanged
 volumeMultiplier=0.5
 concurrencyOverrides {resolutionRule=PreventNew}
}
